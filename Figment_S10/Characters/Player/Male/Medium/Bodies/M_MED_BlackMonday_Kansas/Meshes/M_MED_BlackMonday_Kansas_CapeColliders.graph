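# Physics asset of cape colliders for the BlackMonday Kansas medium male player body, named as collision shapes for its cape.
# Each body is a single capsule covering the pelvis, spine chain, neck, clavicles, upper arms, thighs, calves and feet; forearms, hands and head have none.
# Constraints join neighboring bodies into a hierarchy rooted at the pelvis, with all swing and twist motion free, so they impose no angular limits.
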PhysicsAsset /Figment_S10/Characters/Player/Male/Medium/Bodies/M_MED_BlackMonday_Kansas/Meshes/M_MED_BlackMonday_Kansas_CapeColliders
bodies {
 calf_l [{shapeType=Capsule, length=31.242, radius=11.9116, center=<-19.4843 -0.5441 -0.2294>, rotation=<rot 0 94.0793 -89.4659>}]
 calf_r [{shapeType=Capsule, length=31.5362, radius=11.5971, center=<19.4486 0.3664 0.2406>, rotation=<rot 0 93.6446 -89.3731>}]
 clavicle_l [{shapeType=Capsule, length=8.5761, radius=10.3302, center=<8.351 -1.2184 -2.7958>, rotation=<rot 85.0841 0 165.9572>}]
 clavicle_r [{shapeType=Capsule, length=8.5761, radius=10.3302, center=<-8.351 1.2184 2.7958>, rotation=<rot 85.0841 0 -14.0428>}]
 foot_l [{shapeType=Capsule, length=14.4367, radius=8.1146, center=<-4.2025 6.2288 -0.0237>, rotation=<rot -1.7928 -172.4181 -80.9982>}]
 foot_r [{shapeType=Capsule, length=14.4367, radius=8.2594, center=<4.3963 -6.0338 0.0727>, rotation=<rot -1.8067 -174.7364 -80.9479>}]
 neck_01 [{shapeType=Capsule, length=6.005, radius=10.255, center=<0 2.6796 0.0003>, rotation=<rot 90 0 0>}]
 pelvis [{shapeType=Capsule, length=11.4559, radius=14.3329, center=<-0.1419 -1.2042 1.0154>}]
 spine_01 [{shapeType=Capsule, length=4.2616, radius=13.7349, center=<0.908 -2.2503 -0.0868>, rotation=<rot -0.9479 -90 -0.286>}]
 spine_02 [{shapeType=Capsule, length=5.3566, radius=13.5353, center=<4.0244 0.5685 0>, rotation=<rot 0 -90 0>}]
 spine_03 [{shapeType=Capsule, length=11.8005, radius=13.8229, center=<9.7415 0.3497 0>, rotation=<rot -0.0002 -90 0>}]
 spine_04 [{shapeType=Capsule, length=17.9609, radius=14.8991, center=<7.9544 0.8497 0>, rotation=<rot 0 90 0>}]
 thigh_l [{shapeType=Capsule, length=25.252, radius=11.3318, center=<-19.839 -0.0002 0>, rotation=<rot 90 0 0>}]
 thigh_r [{shapeType=Capsule, length=24.8653, radius=11.1815, center=<19.0128 0 0>, rotation=<rot 90 0 0>}]
 upperarm_l [{shapeType=Capsule, length=17.8261, radius=8.5802, center=<12.1824 -1.5581 -0.6865>, rotation=<rot 90 -1.7899 172.5304>}]
 upperarm_r [{shapeType=Capsule, length=16.3857, radius=7.8839, center=<-12.721 1.2021 0.6993>, rotation=<rot -0.3116 -83.4842 -86.4336>}]
}
constraints [
 {
  bone1Name=upperarm_l
  bone2Name=clavicle_l
  swing1LimitDegrees=45
  swing1Motion=Free
  swing2LimitDegrees=45
  swing2Motion=Free
  twistLimitDegrees=45
  twistMotion=Free
 }
 {
  bone1Name=upperarm_r
  bone2Name=clavicle_r
  swing1LimitDegrees=45
  swing1Motion=Free
  swing2LimitDegrees=45
  swing2Motion=Free
  twistLimitDegrees=45
  twistMotion=Free
 }
 {
  bone1Name=thigh_l
  bone2Name=pelvis
  swing1LimitDegrees=45
  swing1Motion=Free
  swing2LimitDegrees=45
  swing2Motion=Free
  twistLimitDegrees=45
  twistMotion=Free
 }
 {
  bone1Name=calf_l
  bone2Name=thigh_l
  swing1LimitDegrees=45
  swing1Motion=Free
  swing2LimitDegrees=45
  swing2Motion=Free
  twistLimitDegrees=45
  twistMotion=Free
 }
 {
  bone1Name=thigh_r
  bone2Name=pelvis
  swing1LimitDegrees=45
  swing1Motion=Free
  swing2LimitDegrees=45
  swing2Motion=Free
  twistLimitDegrees=45
  twistMotion=Free
 }
 {
  bone1Name=calf_r
  bone2Name=thigh_r
  swing1LimitDegrees=45
  swing1Motion=Free
  swing2LimitDegrees=45
  swing2Motion=Free
  twistLimitDegrees=45
  twistMotion=Free
 }
 {
  bone1Name=spine_04
  bone2Name=spine_03
  swing1LimitDegrees=45
  swing1Motion=Free
  swing2LimitDegrees=45
  swing2Motion=Free
  twistLimitDegrees=45
  twistMotion=Free
 }
 {
  bone1Name=spine_03
  bone2Name=spine_02
  swing1LimitDegrees=45
  swing1Motion=Free
  swing2LimitDegrees=45
  swing2Motion=Free
  twistLimitDegrees=45
  twistMotion=Free
 }
 {
  bone1Name=spine_02
  bone2Name=spine_01
  swing1LimitDegrees=45
  swing1Motion=Free
  swing2LimitDegrees=45
  swing2Motion=Free
  twistLimitDegrees=45
  twistMotion=Free
 }
 {
  bone1Name=spine_01
  bone2Name=pelvis
  swing1LimitDegrees=45
  swing1Motion=Free
  swing2LimitDegrees=45
  swing2Motion=Free
  twistLimitDegrees=45
  twistMotion=Free
 }
 {
  bone1Name=neck_01
  bone2Name=spine_04
  swing1LimitDegrees=45
  swing1Motion=Free
  swing2LimitDegrees=45
  swing2Motion=Free
  twistLimitDegrees=45
  twistMotion=Free
 }
 {
  bone1Name=foot_r
  bone2Name=calf_r
  swing1LimitDegrees=45
  swing1Motion=Free
  swing2LimitDegrees=45
  swing2Motion=Free
  twistLimitDegrees=45
  twistMotion=Free
 }
 {
  bone1Name=foot_l
  bone2Name=calf_l
  swing1LimitDegrees=45
  swing1Motion=Free
  swing2LimitDegrees=45
  swing2Motion=Free
  twistLimitDegrees=45
  twistMotion=Free
 }
 {
  bone1Name=clavicle_l
  bone2Name=spine_04
  swing1LimitDegrees=45
  swing1Motion=Free
  swing2LimitDegrees=45
  swing2Motion=Free
  twistLimitDegrees=45
  twistMotion=Free
 }
 {
  bone1Name=clavicle_r
  bone2Name=spine_04
  swing1LimitDegrees=45
  swing1Motion=Free
  swing2LimitDegrees=45
  swing2Motion=Free
  twistLimitDegrees=45
  twistMotion=Free
 }
]
